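; Slone clone NPC pawn that stays hidden until revealed, then becomes alerted at Threatened level.
; On the server, a gameplay event sets ShowPawn, whose RepNotify runs UnhideClone, while a fallback timer calls UnhideClone directly as a failsafe.
; UnhideClone runs once, spawning Niagara VFX on remotes and unhiding the actor.
; Staggered delays then hack in the overhead "!" cue before properly reactivating the alertness system.
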(event EventBeginPlay
  ;[1 1 1] Failsafe if Slone clones are not unhidden for X reason.
  (|CallParentFunction)
  (Networking|SwitchHasAuthority
    (:Authority
      (Utilities|FlowControl|Sequence
        (:then_0
          (Ability|Async|WaitGameplayEventToActor 0 (Variables|Default|GetEventTag_ListenToUnhide) true
            (:EventReceived
              (CallFunction|ShowClone))))
        (:then_1
          (Utilities|FlowControl|Delay (ScalableFloat|GetValueatLevel (Variables|Default|GetUnhideCloneFallbackTimer)))
          (CallFunction|UnhideClone))))))

(event Collision|EventActorBeginOverlap
  (params (OtherActor :type Actor%20Object%20Reference))
  (|CallParentFunction OtherActor) ; This node is disabled and will not be called.\nDrag off pins to build functionality.
)

(event EventTick
  (params (DeltaSeconds :type Float%20%28single-precision%29))
  (|CallParentFunction DeltaSeconds) ; This node is disabled and will not be called.\nDrag off pins to build functionality.
)

(event Custom|ShowClone
  (Variables|Default|SetShowPawn true))

(event Custom|UnhideClone
  (Utilities|FlowControl|DoOnce false
    (:Completed
      (Networking|SwitchHasAuthority
        (:Authority
          (Utilities|FlowControl|Delay 0.1))
        (:Remote
          (Niagara|SpawnSystemAttached (Variables|Default|GetVFXSystemOnSpawn) (Variables|Character|GetMesh) "root" "0, 0, 0" "0, 0, 0" "SnapToTargetIncludingScale" true true "None" false)))
      (Rendering|SetActorHiddenInGame false)
      ;[1 1 1] Small Hack to Spawn With ! on top of their heads, then make sure to properly re-activtate the system.
      (Utilities|FlowControl|Delay 0.5)
      (CallFunction|SetAlertedStateGameplayCue true true)
      (Utilities|FlowControl|Delay 2.0)
      (Variables|Default|SetUseAlertness true)
      (Variables|Default|SetAlertLevel "Threatened")
      (CallFunction|ActivateAlertDBNOCleanUp))))

(fn OnRep_ShowPawn
  (if (Variables|Default|GetShowPawn)
    (CallFunction|UnhideClone)))

(fn ConstructionScript
  (|CallParentFunction))
